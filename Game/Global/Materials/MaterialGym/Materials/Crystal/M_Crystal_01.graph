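# M_Crystal_01 is a crystal surface material in MaterialGym, flagged for use on skeletal meshes and mesh deformers.
# It builds base material attributes from parameters for color, metallic, specular and roughness, plus a normal map behind the UseNormalMap static switch.
# Those attributes feed the shared MF_Crystal function at a constant alpha of 1, leaving the crystal-specific look to that function.
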Material /Game/Global/Materials/MaterialGym/Materials/Crystal/M_Crystal_01
graph {
 properties {
  bUseMaterialAttributes=true
  bUsedWithMeshDeformer=true
  bUsedWithSkeletalMesh=true
  pixelDepthOffsetMode=PDOM_Legacy
  shadingModel=MSM_FromMaterialExpression
  usedShadingModels=Default Lit
 }
 nodes [
  n1 = TextureSampleParameter2D() {parameterName=NormalMap, samplerType=SAMPLERTYPE_Normal, texture=/Engine/EngineMaterials/BaseFlattenNormalMap}
  n2 = Constant3Vector() {constant=<rgba 0 0 1 0>}
  MaterialFunctionCall!(
   Material=MakeMaterialAttributes!(
    BaseColor=VectorParameter() {parameterName=Color, defaultValue=<rgba 0.18 0.18 0.18 1>}.RGB
    Metallic=ScalarParameter() {parameterName=Metallic}
    Specular=ScalarParameter() {defaultValue=0.5, parameterName=Specular}
    Roughness=ScalarParameter() {parameterName=Roughness}
    Anisotropy
    EmissiveColor
    Opacity
    OpacityMask
    Normal=StaticSwitchParameter(True=@n1.RGB, False=@n2) {parameterName=UseNormalMap}
    Tangent
    WorldPositionOffset
    SubsurfaceColor
    ClearCoat
    ClearCoatRoughness
    AmbientOcclusion
    Refraction
    CustomizedUVs_0
    CustomizedUVs_1
    CustomizedUVs_2
    CustomizedUVs_3
    CustomizedUVs_4
    CustomizedUVs_5
    CustomizedUVs_6
    CustomizedUVs_7
    PixelDepthOffset
    ShadingModel
    Displacement
   )
   Alpha=Constant() {r=1}
  ) {materialFunction=/Game/Global/MaterialFunctions/TechArt/MF_Crystal}
 ]
}
